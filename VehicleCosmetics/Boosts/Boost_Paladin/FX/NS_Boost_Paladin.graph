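# Niagara system for the Paladin vehicle boost cosmetic, trailing ribbons and smoke from skeletal mesh sockets, shifted by a car offset.
# The ribbon emitter feeds six ribbon renderers from shared particles, using per-trail color, width and material variables.
# The smoke emitter merges three spawn groups into one emitter with three sprite renderers, partitioned across ports.
# Painted-variant modules tint both emitters, and bIsPainted switches several trail colors between default values and paint-driven colors.
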NiagaraSystem /VehicleCosmetics/Boosts/Boost_Paladin/FX/NS_Boost_Paladin
dependencies {
 usedDataInterfaces [
  &NiagaraDataInterfaceCurve
  &NiagaraDataInterfaceSpriteRendererInfo
  &NiagaraDataInterfaceColorCurve
  &NiagaraDataInterfaceArrayFloat
  &NiagaraDataInterfaceSkeletalMesh
 ]
 usedDynamicInputs []
 usedModules [
  &self:SystemSpawnScript.SystemScriptSource.SystemScriptGraph.NiagaraNodeAssignment_0.SetVariables_#1
  &SystemState
 ]
}
emitters {
 BCKP_Ribbons {
  emitterUpdateScript [
   &EmitterState()
   &VC_FilteredSocketCount()
   &VC_RateQualityHandler(High Spawn Rate=50, Low Spawn Rate=30, Medium Spawn Rate=40, Use Custom Rate?=true)
   &VC_SpawnRibbon_v1(Number of Ribbons=$Emitter.PortCount, SpawnRate=$Emitter.Spawn Rate)
  ]
  particleSpawnScript [
   &VC_PaintedVariantsOverride(
    Black=<rgba 0.005 0.005 0.005 1>
    BurntSienna=<rgba 0.1146 0.0354 0.0141 1>
    Cobalt=<rgba 0 0.0822 0.625 1>
    Crimson=<rgba 0.4635 0 0.0126 1>
    DefaultColor=<rgba 0.1266 0.0303 0.4375 5>
    ForestGreen=<rgba 0 0.1562 0.0079 1>
    Grey=<rgba 0.01 0.01 0.01 1>
    Lime=<rgba 0.0635 0.1 0 1>
    Orange=<rgba 0.4479 0.0747 0 1>
    Pink=<rgba 0.7135 0.0256 0.5765 1>
    Purple=<rgba 0.0309 0 0.2552 1>
    R1 Black:LinearColor=<rgba 0.0191 0.0191 0.0191 0.2>
    R1 BurntSienna:LinearColor=<rgba 1 0.251 0.05 0.05>
    R1 Cobalt:LinearColor=<rgba 0 0.4 1.2 0.13>
    R1 Crimson:LinearColor=<rgba 2.2 0.22 0.2737 0.04>
    R1 ForestGreen:LinearColor=<rgba 0 1.3 0.0657 0.04>
    R1 Grey:LinearColor=<rgba 0.2031 0.2031 0.2031 0.08>
    R1 Lime:LinearColor=<rgba 0.2943 0.5885 0 0.08>
    R1 Orange:LinearColor=<rgba 2.2 0.44 0 0.08>
    R1 Pink:LinearColor=<rgba 1 0.15 0.6909 0.1>
    R1 Purple:LinearColor=<rgba 0.1667 0 1 0.1>
    R1 Saffron:LinearColor=<rgba 1.2 0.92 0 0.05>
    R1 SkyBlue:LinearColor=<rgba 0 1.3308 2 0.1>
    R1 TitaniumWhite:LinearColor=<rgba 0.8 0.8 0.8 0.2>
    Saffron=<rgba 0.1057 0.1094 0 1>
    SkyBlue=<rgba 0.0002 0.2056 0.2552 1>
    TitaniumWhite=<rgba 0.2969 0.2969 0.2969 1>
    Use R1 Painted Overrides=true
   )
   &VC_InitializeRibbon_v1(
    Color=$Particles.Color
    Custom Offset=true
    Enable Skeletal Position=true
    Lifetime=1
    Local Vector Offset=&AddVectorToPosition(Position=<-40 0 0>, Vector=$User.DTCarOffset)
    Offset Amount=1
    Ribbon Width:NiagaraFloat=&Multiply_Float(A=58, B=4)
   )
   &AddVelocity(Velocity=<-1500 0 0>) {module=AddVelocity001}
  ]
  particleUpdateScript [
   &ParticleState()
   &InheritSourceMovement(
    Applied Position Scale=&VectorFromFloat(Value=&FloatFromCurve(FloatCurve=DataInterface() {propertyValues {Curve {keys [
          {interpMode=RCIM_Cubic, tangentMode=RCTM_Break, time=0, value=1}
          {arriveTangent=-2.9025, interpMode=RCIM_Cubic, tangentMode=RCTM_Break, time=1, value=0}
         ]}}}))
    ApplyForce=false
    ApplyPosition=true
    ApplyPositionFalloff=false
   )
   &SolveForcesAndVelocity(Clamp Velocity=false, Limit Acceleration=false)
   &VC_UpdateRibbon_v1(Enable Skeletal Position=true)
   &ScaleColor(
    Scale Alpha=&FloatFromCurve(FloatCurve=DataInterface() {propertyValues {Curve {keys [<at 0 0>,<at 0.1 1.35>,<at 1 0>]}}})
    Scale RGB=&VectorFromFloat(Value=&FloatFromCurve(FloatCurve=DataInterface() {propertyValues {Curve {keys [<at 0 0>,<at 0.1 0.9>,<at 1 0>]}}}, Scale Curve=1.5))
    ScaleA=true
    ScaleRGB=true
   )
   &ScaleRibbonWidth(Ribbon Width Scale=&FloatFromCurve(FloatCurve=DataInterface() {propertyValues {Curve {keys [<at 0 0.9>,<at 1 0>]}}}))
   &DynamicMaterialParameters(
    Index 0 Param 1=2
    Index 0 Param 2=2
    Index 0 Param 3=1
    Index 0 Param 4=-1
    Module.Index0Float0Write=true
    Module.Index0Float1Write=true
    Module.Index0Float2Write=true
    Module.Index0Float3Write=true
   )
   /SetVariables(
    Particles.DMP_FifthTrail:Vector4f=&MakeVector4(
     X=0.8
     Y=&SelectFloatFromArray(Direct Array Index=$Particles.SpawnGroup, Float Selection Array=DataInterface() {propertyValues {FloatData [0,0.5]}})
     Z=0.4
     W=0
    )
    Particles.DMP_FourthTrail:Vector4f=&MakeVector4(X=0, Y=1, Z=0, W=1)
    Particles.DMP_SecondTrail:Vector4f=&MakeVector4(X=1, Y=1, Z=6, W=0.04)
    Particles.DMP_ThirdTrail:Vector4f=&MakeVector4(X=1, Y=2, Z=1, W=-1)
    Particles.FifthTrail:LinearColor=&MakeCustomLinearColorFromBool(
     Bool=$User.bIsPainted
     True Color=&VC_DynamicScaleColor(
      Color to Scale:LinearColor=$User.Color_Ribbon_5
      RGB Scale:Vector3f=&VectorFromFloat(Value=&FloatFromCurve(FloatCurve=DataInterface() {propertyValues {Curve {keys [<at 0.1 1 RCIM_Cubic>,<at 1 0 RCIM_Cubic>]}}}, Scale Curve=10))
      Alpha Scale:NiagaraFloat=&FloatFromCurve(FloatCurve=DataInterface() {propertyValues {Curve {keys [<at 0.5 1 RCIM_Cubic>,<at 1 0 RCIM_Cubic>]}}}, Scale Curve=5)
     )
     False Color=&ColorFromCurve(ColorCurve=DataInterface() {propertyValues {
        AlphaCurve {keys [<at 0.5329 5 RCIM_Cubic>,<at 0.6473 0 RCIM_Cubic>]}
        BlueCurve {keys [<at 0 0 RCIM_Cubic>,<at 0.3983 0>]}
        GreenCurve {keys [<at 0 0 RCIM_Cubic>,<at 0.3983 0.0293>]}
        RedCurve {keys [<at 0 7 RCIM_Cubic>,<at 0.3983 5.5>]}
       }})
    )
    Particles.FourthTrail:LinearColor=&MakeCustomLinearColorFromBool(
     Bool=$User.bIsPainted
     True Color=&VC_DynamicScaleColor(
      Color to Scale:LinearColor=$Particles.R1 Color
      RGB Scale:Vector3f=&VectorFromFloat(Value=2)
      Alpha Scale:NiagaraFloat=&FloatFromCurve(FloatCurve=DataInterface() {propertyValues {Curve {keys [
           <at 0 0 RCIM_Cubic>
           {arriveTangent=1.7573, interpMode=RCIM_Cubic, leaveTangent=1.7573, time=0.2, value=1.6}
           {arriveTangent=-1.5, interpMode=RCIM_Cubic, leaveTangent=-1.5, time=0.3492, value=0.6137}
           <at 1 0.4 RCIM_Cubic>
          ]}}})
     )
     False Color=&ColorFromCurve(ColorCurve=DataInterface() {propertyValues {
        AlphaCurve {keys [{arriveTangent=-2.5606, interpMode=RCIM_Cubic, leaveTangent=-2.5606, tangentMode=RCTM_User, time=0.2125, value=1}, <at 0.4413 0 RCIM_Cubic>]}
        BlueCurve {keys [
          <at 0 0 RCIM_Cubic>
          <at 0.2735 0.03>
          <at 0.4391 0.0355>
          {arriveTangent=-0.0422, interpMode=RCIM_Cubic, leaveTangent=-0.0422, time=0.7801, value=0}
          <at 0.9842 0>
         ]}
        GreenCurve {keys [
          <at 0 0.0833 RCIM_Cubic>
          <at 0.2735 0.3975>
          <at 0.4391 0.4761>
          {arriveTangent=-0.5593, interpMode=RCIM_Cubic, leaveTangent=-0.5593, time=0.7801, value=0.0833}
          <at 0.9842 0>
         ]}
        RedCurve {keys [
          <at 0 0.5 RCIM_Cubic>
          <at 0.2735 1.5>
          <at 0.4391 1.2>
          {arriveTangent=-0.7035, interpMode=RCIM_Cubic, leaveTangent=-0.7035, time=0.7801, value=1}
          <at 0.9842 1>
         ]}
       }})
    )
    Particles.SecondTrail:LinearColor=&MakeCustomLinearColorFromBool(
     Bool=$User.bIsPainted
     True Color=&VC_DynamicScaleColor(
      Color to Scale:LinearColor=$User.Color_Ribbon_2
      RGB Scale:Vector3f=&VectorFromFloat(Value=&FloatFromCurve(FloatCurve=DataInterface() {propertyValues {Curve {keys [<at 0 1>,<at 1 0>]}}}))
      Alpha Scale:NiagaraFloat=1
     )
     False Color=&ColorFromCurve(ColorCurve=DataInterface() {propertyValues {
        AlphaCurve {keys [<at 0 1 RCIM_Cubic>,<at 1 1 RCIM_Cubic>]}
        BlueCurve {keys [
          <at 0.02 0 RCIM_Cubic>
          {arriveTangent=-0.0004, interpMode=RCIM_Cubic, leaveTangent=-0.0004, tangentMode=RCTM_User, time=0.45, value=0}
          <at 1 1 RCIM_Cubic>
         ]}
        GreenCurve {keys [
          <at 0.01 0.11 RCIM_Cubic>
          {arriveTangent=-0.1111, interpMode=RCIM_Cubic, leaveTangent=-0.1111, time=0.45, value=0.1231}
          <at 1 0 RCIM_Cubic>
         ]}
        RedCurve {keys [
          <at 0.01 1 RCIM_Cubic>
          {arriveTangent=-0.0382, interpMode=RCIM_Cubic, leaveTangent=-0.0382, tangentMode=RCTM_User, time=0.45, value=1}
          <at 1 0.3333 RCIM_Cubic>
         ]}
       }})
    )
    Particles.SixthTrail:LinearColor=&MakeCustomLinearColorFromBool(
     Bool=$User.bIsPainted
     True Color=&VC_DynamicScaleColor(
      Color to Scale:LinearColor=$User.PaintedColor
      RGB Scale:Vector3f=<1 1 1>
      Alpha Scale:NiagaraFloat=&FloatFromCurve(FloatCurve=DataInterface() {propertyValues {Curve {keys [<at 0 1>,<at 1 0>]}}})
     )
     False Color=<rgba 1 0.0833 0 1.75>
    )
    Particles.ThirdTrail:LinearColor=&MakeCustomLinearColorFromBool(
     Bool=$User.bIsPainted
     True Color=&VC_DynamicScaleColor(
      Color to Scale:LinearColor=$User.Color_Ribbon_3
      RGB Scale:Vector3f=&VectorFromFloat(Value=&FloatFromCurve(FloatCurve=DataInterface() {propertyValues {Curve {keys [<at 0 1>,<at 0.55 0>]}}}))
      Alpha Scale:NiagaraFloat=1
     )
     False Color=&ColorFromCurve(ColorCurve=DataInterface() {propertyValues {
        AlphaCurve {keys [<at 0 1 RCIM_Cubic>,<at 1 1 RCIM_Cubic>]}
        BlueCurve {keys [<at 0.0043 0>,<at 0.3146 0>,<at 0.4637 0>,<at 1 0>]}
        GreenCurve {keys [<at 0.0043 0>,<at 0.3146 0.1352>,<at 0.4637 0.0167>,<at 1 0.005>]}
        RedCurve {keys [<at 0 1 RCIM_Cubic>,<at 0.0043 1>,<at 0.3146 1>,<at 0.4637 1>,<at 1 0.1>]}
       }})
    )
    Particles.Width_FifthTrail:NiagaraFloat=&FloatFromCurve(FloatCurve=DataInterface() {propertyValues {Curve {keys [<at 0 0.85>,<at 0.2 1.6>,<at 0.4 2>,<at 1 0>]}}}, Scale Curve=50)
    Particles.Width_FourthTrail:NiagaraFloat=&FloatFromCurve(
     FloatCurve=DataInterface() {propertyValues {Curve {keys [
         {interpMode=RCIM_Cubic, tangentMode=RCTM_Break, time=-0.03, value=0}
         {arriveTangent=1.1321, interpMode=RCIM_Cubic, leaveTangent=1.1321, time=0.1491, value=0.5544}
         {interpMode=RCIM_Cubic, leaveTangent=-0.6562, tangentMode=RCTM_Break, time=0.5, value=0.6}
         {interpMode=RCIM_Cubic, tangentMode=RCTM_Break, time=1, value=0.15}
        ]}}}
     Scale Curve=&MakeCustomFloatFromBool(Bool=$User.bIsPainted, True Float=30, False Float=18)
    )
    Particles.Width_SecondTrail:NiagaraFloat=&FloatFromCurve(
     FloatCurve=DataInterface() {propertyValues {Curve {keys [
         {interpMode=RCIM_Cubic, tangentMode=RCTM_Break, time=-0.03, value=0}
         {arriveTangent=1.1321, interpMode=RCIM_Cubic, leaveTangent=1.1321, time=0.1491, value=0.5544}
         {interpMode=RCIM_Cubic, leaveTangent=-0.6562, tangentMode=RCTM_Break, time=0.5, value=0.6}
         {interpMode=RCIM_Cubic, tangentMode=RCTM_Break, time=1, value=0.2}
        ]}}}
     Scale Curve=&MakeCustomFloatFromBool(Bool=$User.bIsPainted, True Float=30, False Float=18)
    )
    Particles.Width_ThirdTrail:NiagaraFloat=&FloatFromCurve(FloatCurve=DataInterface() {propertyValues {Curve {keys [<at 0 0.9>,<at 1 0>]}}}, Scale Curve=220)
   ) {isSetParameters=true}
  ]
  rendererClasses [&NiagaraRibbonRendererProperties]
  renderers [
   {rendererClass=&NiagaraRibbonRendererProperties}
   {rendererClass=&NiagaraRibbonRendererProperties, rendererIndex=1}
   {rendererClass=&NiagaraRibbonRendererProperties, rendererIndex=2}
   {rendererClass=&NiagaraRibbonRendererProperties, rendererIndex=3}
   {rendererClass=&NiagaraRibbonRendererProperties, rendererIndex=4}
   {rendererClass=&NiagaraRibbonRendererProperties, rendererIndex=5}
  ]
 }
 BCKP_Smoke_Emitter {
  emitterUpdateScript [
   &EmitterState(Inactive Response=::NewEnumerator1, Life Cycle Mode=::NewEnumerator1, UseLoopDelay=false)
   &VC_FilteredSocketCount()
   &SpawnRate(SpawnRate=10, Use Spawn Probability=false)
   &SpawnRate(SpawnGroup=1, SpawnRate=10, Use Spawn Probability=false) {module=SpawnRate001}
   &SpawnRate(SpawnGroup=2, SpawnRate=30, Use Spawn Probability=false) {module=SpawnRate002}
  ]
  particleSpawnScript [
   /SetVariables(Particles.TagsSmoke:NiagaraInt32=&SpawnGroup()) {isSetParameters=true}
   &VC_PaintedVariantsOverride(
    Black=<rgba 0 0 0 40>
    BurntSienna=<rgba 0.11 0.034 0.0136 5>
    Cobalt=<rgba 0 0.0024 0.13 6>
    Crimson=<rgba 0.13 0 0.0035 5>
    DefaultColor=<rgba 0 0 0 5>
    ForestGreen=<rgba 0 0.1 0.0051 5>
    Grey=<rgba 0.12 0.12 0.12 5>
    Lime=<rgba 0.065 0.13 0 5>
    Orange=<rgba 0.13 0.0453 0 5>
    Pink=<rgba 0.13 0.0156 0.0884 5>
    Purple=<rgba 0.1149 0 0.16 5>
    R1 Black:LinearColor=<rgba 0 0 0 0.5>
    R1 BurntSienna:LinearColor=<rgba 0.02 0.0062 0.0025 0.5>
    R1 Cobalt:LinearColor=<rgba 0 0.0046 0.03 0.5>
    R1 Crimson:LinearColor=<rgba 0.02 0 0.0005 0.5>
    R1 DefaultColor=<rgba 0.001 0.001 0.001 1>
    R1 ForestGreen:LinearColor=<rgba 0 0.02 0.001 0.5>
    R1 Grey:LinearColor=<rgba 0.02 0.02 0.02 0.5>
    R1 Lime:LinearColor=<rgba 0.0025 0.005 0 0.4>
    R1 Orange:LinearColor=<rgba 0.023 0.008 0 0.5>
    R1 Pink:LinearColor=<rgba 0.02 0.0024 0.0136 0.5>
    R1 Purple:LinearColor=<rgba 0.0082 0 0.02 0.5>
    R1 Saffron:LinearColor=<rgba 0.004 0.004 0 0.502>
    R1 SkyBlue:LinearColor=<rgba 0 0.0133 0.02 0.5>
    R1 TitaniumWhite:LinearColor=<rgba 0.02 0.02 0.02 0.5>
    R2 DefaultColor:LinearColor=<rgba 0 0 0 5>
    Saffron=<rgba 0.08 0.08 0 5>
    SkyBlue=<rgba 0.0001 0.0777 0.08 5>
    TitaniumWhite=<rgba 0.12 0.12 0.12 5>
    Use R1 Painted Overrides=true
    Use R2 Painted Overrides=true
   )
   &InitializeParticle(
    Color Mode=::NewEnumerator0
    Lifetime=/VehicleCosmetics/FX/Niagara/Modules/MergeEmitters/VC_FloatBySpawnGroup(
     SpawnGroup or ExecIndex:NiagaraInt32=0
     CustomInt:NiagaraInt32=0
     MaximumSpawnGroups:NiagaraInt32=3
     Float SG 0:NiagaraFloat=&RandomRangeFloat(Minimum=0.8, Maximum=1)
     Float SG 1:NiagaraFloat=&RandomRangeFloat(Minimum=0.4, Maximum=1)
     Float SG 2:NiagaraFloat=&RandomRangeFloat(Minimum=2, Maximum=4)
     Float SG 3:NiagaraFloat=0
     Float SG 4:NiagaraFloat=0
     Float SG 5:NiagaraFloat=0
    )
    Sprite Rotation Angle=&VC_FloatBySpawnGroup(
     SpawnGroup or ExecIndex:NiagaraInt32=0
     CustomInt:NiagaraInt32=0
     MaximumSpawnGroups:NiagaraInt32=3
     Float SG 0:NiagaraFloat=0
     Float SG 1:NiagaraFloat=0
     Float SG 2:NiagaraFloat=&RandomRangeFloat(Minimum=-1, Maximum=1)
     Float SG 3:NiagaraFloat=0
     Float SG 4:NiagaraFloat=0
     Float SG 5:NiagaraFloat=0
    )
    Sprite Rotation Mode=::NewEnumerator1
    UsePositionOffset=false
    Write Material Random=true
   )
   &PartitionParticles(Number Of Partitions=$Emitter.PortCount)
   &SkeletalMeshLocation(
    Bone / Socket Index=$Particles.PartitionParticles.Partition
    Bone / Socket Sampling Mode=::NewEnumerator5
    Interpolate Mesh Sampling=false
    MaskSpawnGroup=false
    Mesh Sampling Type=::NewEnumerator6
    Sampled Position Offset=/VehicleCosmetics/FX/Niagara/Modules/MergeEmitters/VC_VectorBySpawnGroup(
     SpawnGroup or ExecIndex:NiagaraInt32=0
     CustomInt:NiagaraInt32=0
     MaximumSpawnGroups:NiagaraInt32=3
     Vector Group 0:Vector3f=&AddVectorToPosition(Position=<-110 0 0>, Vector=$User.DTCarOffset)
     Vector Group 1:Vector3f=&AddVectorToPosition(Position=<-50 0 0>, Vector=$User.DTCarOffset)
     Vector Group 2:Vector3f=&AddVectorToPosition(Position=<-20 0 0>, Vector=$User.DTCarOffset)
     Vector Group 3:Vector3f=<0 0 0>
     Vector Group 4:Vector3f=<0 0 0>
     Vector Group 5:Vector3f=<0 0 0>
    )
    Skeletal Mesh=$System.SkeletalMesh
   )
   /CascadeToNiagaraConverter/NiagaraScripts/ModuleScripts/SpriteLockAxis(Lock Axis Type:ENiagara_SpriteLockAxis=::NewEnumerator6, UseAlignmentCoordinateSpace=true, UseFacingCoordinateSpace=true)
   &AddVelocity(Velocity=&VC_VectorBySpawnGroup(
     SpawnGroup or ExecIndex:NiagaraInt32=0
     CustomInt:NiagaraInt32=0
     MaximumSpawnGroups:NiagaraInt32=3
     Vector Group 0:Vector3f=&RandomRangeVector(Minimum=<-50 0 0>, Maximum=<-10 0 0>)
     Vector Group 1:Vector3f=&RandomRangeVector(Minimum=<-20 0 0>, Maximum=<-90 0 0>)
     Vector Group 2:Vector3f=&RandomRangeVector(Minimum=<-100 0 0>, Maximum=<-80 0 0>)
     Vector Group 3:Vector3f=<0 0 0>
     Vector Group 4:Vector3f=<0 0 0>
     Vector Group 5:Vector3f=<0 0 0>
    ))
   &Color(Color=/VehicleCosmetics/FX/Niagara/Modules/MergeEmitters/VC_Dynamic_ColorBySpawnGroup(
     Color SG 0:LinearColor=$Particles.Color
     MaximumSpawnGroups:NiagaraInt32=3
     Color SG 1:LinearColor=$Particles.Color
     Color SG 2:LinearColor=$Particles.Color
     Color SG 3:LinearColor=<rgba 1 1 1 1>
     Color SG 4:LinearColor=<rgba 1 1 1 1>
    ))
   /SetVariables(Particles.Size_FirstSmoke:Vector2f=&Multiply_Vector2D_ByFloat(Float=3, Vector2D=<70 30>)) {isSetParameters=true}
   &ShapeLocation(
    MaskSpawnGroup=false
    Override Seed=false
    Sphere Radius=&RandomRangeFloat(Minimum=0, Maximum=8)
    UseOwnerScale=true
   )
  ]
  particleUpdateScript [
   &ParticleState()
   &SolveForcesAndVelocity(Clamp Velocity=false, Limit Acceleration=false)
   /SetVariables(
    Particles.FirstSmokeColor:LinearColor=&MultiplyLinearColor(
     Linear Color=$Particles.Color
     Scale Factor=&MakeVector4FromVectorAndFloat(
      "Vector (RGB)"=<1 1 1>
      "Float (Alpha)"=&FloatFromCurve(FloatCurve=DataInterface() {propertyValues {Curve {
          defaultValue=1
          keys [
           {arriveTangent=9.0725, interpMode=RCIM_Cubic, leaveTangent=9.0725, tangentMode=RCTM_User, time=0, value=0}
           {arriveTangent=-2.5249, interpMode=RCIM_Cubic, leaveTangent=-2.5249, tangentMode=RCTM_User, time=0.1408, value=0.9847}
           {arriveTangent=-0.0405, interpMode=RCIM_Cubic, leaveTangent=-0.0405, tangentMode=RCTM_User, time=1.1203, value=-1.2761}
          ]
         }}})
     )
    )
    Particles.SecondSmokeColor:LinearColor=&MultiplyLinearColor(
     Linear Color=$Particles.Color
     Scale Factor=&MakeVector4FromVectorAndFloat(
      "Vector (RGB)"=<1 1 1>
      "Float (Alpha)"=&FloatFromCurve(FloatCurve=DataInterface() {propertyValues {Curve {
          defaultValue=1
          keys [{interpMode=RCIM_Cubic, tangentMode=RCTM_User, time=0, value=1}, {interpMode=RCIM_Cubic, tangentMode=RCTM_User, time=1, value=0}]
         }}})
     )
    )
    Particles.Size_SecondSmoke:Vector2f=&Multiply_Vector2D_ByFloat(
     Float=&FloatFromCurve(FloatCurve=DataInterface() {propertyValues {Curve {keys [<at 0 0.85>,<at 1 1.4>]}}}, Scale Curve=2)
     Vector2D=<50 60>
    )
    Particles.Size_ThirdSmoke:Vector2f=&Multiply_Vector2D_ByFloat(
     Float=&FloatFromCurve(
      FloatCurve=DataInterface() {propertyValues {Curve {
         defaultValue=0
         keys [
          {arriveTangent=0.643, interpMode=RCIM_Cubic, leaveTangent=0.643, tangentMode=RCTM_User, time=0, value=0.7}
          {arriveTangent=-0.1362, interpMode=RCIM_Cubic, leaveTangent=-0.1362, tangentMode=RCTM_User, time=0.4275, value=1}
          {arriveTangent=0.0619, interpMode=RCIM_Cubic, leaveTangent=0.0619, tangentMode=RCTM_User, time=1, value=0.2}
         ]
        }}}
      Scale Curve=1.6
     )
     Vector2D=<50 50>
    )
   ) {isSetParameters=true}
   &SubUVAnimation(
    "Sprite Renderer "=DataInterface() {propertyValues {SpriteRenderer=&self:NiagaraEmitter_0.NiagaraSpriteRendererProperties_5}}
    Start Frame Range Override=16
    UseEndFrame=false
    UseStartFrame=false
   )
  ]
  rendererClasses [&NiagaraSpriteRendererProperties]
  renderers [
   {rendererClass=&NiagaraSpriteRendererProperties}
   {rendererClass=&NiagaraSpriteRendererProperties, rendererIndex=1}
   {rendererClass=&NiagaraSpriteRendererProperties, rendererIndex=2}
  ]
 }
}
summary {
 systemName=NS_Boost_Paladin
 emitters [
  {emitterName=BCKP_Smoke_Emitter, rendererClasses [&NiagaraSpriteRendererProperties]}
  {emitterName=BCKP_Ribbons, rendererClasses [&NiagaraRibbonRendererProperties]}
 ]
 userVariables [
  User.bIsFrontendPreview {type=NiagaraBool}
  User.bIsPainted {type=NiagaraBool}
  User.Color_Ribbon_2 {type=LinearColor, description="", defaultValue {struct=LinearColor, value=<rgba 1 1 1 1>}}
  User.Color_Ribbon_3 {type=LinearColor, description="", defaultValue {struct=LinearColor, value=<rgba 1 1 1 1>}}
  User.Color_Ribbon_4 {type=LinearColor, description="", defaultValue {struct=LinearColor, value=<rgba 1 1 1 1>}}
  User.Color_Ribbon_5 {type=LinearColor, description="", defaultValue {struct=LinearColor, value=<rgba 1 1 1 1>}}
  User.ColorVariantID {type=NiagaraInt32}
  User.DTCarOffset {type=Vector3f, description="", defaultValue {struct=Vector3f, value=<0 0 0>}}
  User.PaintedColor {type=LinearColor, description="", defaultValue {struct=LinearColor, value=<rgba 1 1 1 1>}}
  User.PaintedScalar {type=NiagaraFloat, defaultValue=1, description=""}
  User.PortCount {type=NiagaraInt32, defaultValue=1}
 ]
}
