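# Translucent unlit Niagara material for the small trail of the Paladin vehicle boost.
# A panning trail texture, offset by a dynamic parameter, is shaped by U-gradient falloffs and particle alpha.
# Panning blob noise and a ribbon tip mask are subtracted to erode it, with a Bloom? static switch replacing that erosion with a constant.
# Camera depth fade and saturation finish the mask, while particle color feeds a shared emissive adjustment function.
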Material /VehicleCosmetics/Boosts/Boost_Paladin/Materials/M_Boost_Paladin_SmallTrail_Opt
graph {
 properties {
  bAutomaticallySetUsageInEditor=false
  bUsedWithNiagaraMeshParticles=true
  bUsedWithNiagaraRibbons=true
  bUsedWithNiagaraSprites=true
  blendMode=BLEND_Translucent
  shadingModel=MSM_Unlit
  usedShadingModels=Unlit
 }
 nodes [
  n1 = Constant() {r=0.1}
  n2 = AppendVector(A=DynamicParameter() {paramNames [speed, offset, EndDissolve, UvDistortAmount]}.offset, B=Constant())
  n3 = TextureCoordinate() {uTiling=2}
  n4 = Constant() {r=0.8}
  n5 = Time()
  n6 = TextureSampleParameter2D(UVs=Panner(Coordinate=Add(A=@n2, B=@n3), Time=Multiply(A=@n4, B=@n5)) {speedX=-1.5}) {parameterName=Texture_v1, texture=/DealershipSharedLibrary/FX/Textures/TexturePackage/T_VCVFX_SmallTrailTest_RGB}
  MaterialExpressionVertexColor_4 = ParticleColor()
  n7 = NamedRerouteUsage()
  n8 = Constant() {r=1.3}
  n9 = Multiply(
   A=Multiply(A=Power(Base=@n1, Exp=@n6.RGB), B=@MaterialExpressionVertexColor_4.A)
   B=Power(Base=Multiply(A=@n7, B=@n8), Exp=Constant() {r=1.6})
  )
  n10 = Power(Base=OneMinus(None=NamedRerouteUsage()), Exp=Constant() {r=1.5})
  n11 = Multiply(A=Multiply(A=Multiply(A=@n9, B=@n10), B=Constant() {r=3}), B=@n6.RGB)
  n12 = Panner(Coordinate=TextureCoordinate() {vTiling=0.0125}, Time=Multiply(A=Constant() {r=0.4}, B=Time())) {speedY=1}
  n13 = TextureSample(UVs=Panner(Coordinate=TextureCoordinate() {uTiling=1.4, vTiling=0.7}) {speedX=-2, speedY=0.25}) {texture=/DealershipSharedLibrary/FX/Textures/Noise/T_VCVFX_BlobNoiseGlow_A}
  n14 = TextureCoordinate()
  n15 = NamedRerouteUsage()
  n16 = Constant() {r=12}
  n17 = Multiply(
   A=TextureSample(UVs=Panner(Coordinate=@n14) {speedX=-1, speedY=0.5}) {texture=&T_VCVFX_BlobNoiseGlow_A}.RGB
   B=Multiply(A=Power(Base=@n15, Exp=@n16), B=Constant() {r=200})
  )
  n18 = StaticSwitchParameter(
   True=Constant()
   False=Add(
    A=TextureSample(UVs=@n12) {texture=/DealershipSharedLibrary/FX/Textures/Gradient_Masks/T_VCVFX_Ribbon_Tip_Mask}.RGB
    B=Multiply(A=@n13.RGB, B=@n17)
   )
  ) {parameterName=Bloom?}
  n19 = OneMinus(None=NamedRerouteUsage())
  n20 = Constant() {r=2}
  n21 = Multiply(A=Multiply(A=Subtract(A=@n11, B=@n18), B=Constant() {r=1.3}), B=Multiply(A=Power(Base=@n19, Exp=@n20), B=Constant() {r=20}))
  n22 = Constant()
  n23 = Constant()
  MaterialExpressionConstantClamp_15 = Saturate(None=Multiply(
    A=@n21
    B=MaterialFunctionCall(Fade Length=@n22, Fade Offset=@n23) {materialFunction=&CameraDepthFade}.Result
   ))
  MaterialFunctionCall(Emissive=@MaterialExpressionVertexColor_4.RGB) {materialFunction=&MF_FX_EmissiveAdjustments}
  Multiply(A=@n21, B=Constant3Vector() {constant=<rgba 1 1 1 0>})
  n24 = ComponentMask(None=TextureCoordinate()) {r=true}
  NamedRerouteDeclaration(None=@n24) {name=Mask_R, nodeColor=<rgba 0 1 0.7188 1>}
  Reroute(None=@n24)
 ]
}
