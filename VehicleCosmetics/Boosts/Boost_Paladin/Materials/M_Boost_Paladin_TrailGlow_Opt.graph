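# Translucent unlit Niagara material for the glow on the Paladin boost trail.
# Two oppositely rotated, panning anisotropic textures, split at mid-V, carve streaks out of a scrolling glow texture.
# A scrolling trail texture, U fade-in, inverted cloud mask and dynamic AlphaMult shape the result, which particle color tints for emissive adjustment.
# A separate U-based end fade, boosted by panning noise, further masks that untinted result.
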Material /VehicleCosmetics/Boosts/Boost_Paladin/Materials/M_Boost_Paladin_TrailGlow_Opt
graph {
 properties {
  bAutomaticallySetUsageInEditor=false
  bUsedWithNiagaraMeshParticles=true
  bUsedWithNiagaraRibbons=true
  bUsedWithNiagaraSprites=true
  blendMode=BLEND_Translucent
  refractionMethod=RM_PixelNormalOffset
  shadingModel=MSM_Unlit
  usedShadingModels=Unlit
 }
 nodes [
  MaterialExpressionVertexColor_1 = ParticleColor()
  n1 = TextureCoordinate()
  n2 = TextureSample(UVs=Panner(Coordinate=@n1) {speedX=-4}) {texture=/DealershipSharedLibrary/FX/Textures/TexturePackage/T_VCVFX_Paladin_TrailGlow_RGB}
  n3 = Panner(Coordinate=MaterialFunctionCall(
    UVs=TextureCoordinate() {uTiling=-1, vTiling=2}
    Rotation Center=Constant2Vector() {g=0.5, r=0.5}
    "Rotation Angle (0-1)"=Constant() {r=0.02}
   ) {materialFunction=&CustomRotator}.Rotated Values) {speedX=1}
  n4 = ComponentMask(None=TextureCoordinate()) {g=true}
  n5 = ConstantBiasScale(None=@n4) {bias=0.5, scale=1}
  n6 = ConstantBiasScale(None=OneMinus(None=@n4)) {bias=0.5, scale=1}
  n7 = Panner(Coordinate=MaterialFunctionCall(
    UVs=TextureCoordinate() {uTiling=-1, vTiling=2}
    Rotation Center=Constant2Vector() {g=0.5, r=0.5}
    "Rotation Angle (0-1)"=Constant() {r=-0.02}
   ) {materialFunction=&CustomRotator}.Rotated Values) {speedX=1}
  n8 = Add(
   A=Multiply(A=TextureSample(UVs=@n3) {texture=/DealershipSharedLibrary/FX/Textures/Gradient_Masks/T_VCVFX_Boost_Anisotropic_A}.RGB, B=Floor(None=@n5))
   B=Multiply(A=Floor(None=@n6), B=TextureSample(UVs=@n7) {texture=&T_VCVFX_Boost_Anisotropic_A}.RGB)
  )
  n9 = Constant() {r=3}
  n10 = Multiply(A=Multiply(A=Power(Base=@n8, Exp=@n9), B=OneMinus(None=@n2.G)), B=@n2.R)
  MaterialExpressionConstantClamp_5 = Saturate(None=Multiply(A=@n10, B=Constant() {r=70}))
  MaterialExpressionConstantClamp_4 = Saturate(None=Subtract(A=Max(A=@n2.R, B=@n10), B=@MaterialExpressionConstantClamp_5))
  n11 = ComponentMask(None=TextureCoordinate()) {r=true}
  n12 = ScalarParameter() {defaultValue=0.75, parameterName=FadeInMaskPower}
  MaterialExpressionConstantClamp_3 = Saturate(None=Multiply(A=ScalarParameter() {defaultValue=50, parameterName=FadeInMaskMult}, B=Power(Base=@n11, Exp=@n12)))
  n13 = Panner(Coordinate=@n1) {speedX=1}
  MaterialExpressionConstantClamp_2 = Saturate(None=Multiply(
    A=@MaterialExpressionConstantClamp_3
    B=TextureSample(UVs=@n13) {texture=/DealershipSharedLibrary/FX/Textures/TexturePackage/T_VCVFX_Paladin_Trail_RGB}.R
   ))
  n14 = Multiply(
   A=OneMinus(None=TextureSample() {texture=/DealershipSharedLibrary/FX/Textures/Gradient_Masks/T_VCVFX_boost_Space_Tear_SquareCloud_Mask}.RGB)
   B=@MaterialExpressionConstantClamp_2
  )
  n15 = Multiply(
   A=Multiply(A=DynamicParameter() {paramNames [AlphaMult,Emissive,ColorAdd,TextureSpeed]}.AlphaMult, B=@n14)
   B=@MaterialExpressionConstantClamp_4
  )
  MaterialExpressionConstantClamp_1 = Saturate(None=Multiply(A=@MaterialExpressionVertexColor_1.RGB, B=@n15))
  MaterialFunctionCall(Emissive=@MaterialExpressionConstantClamp_1) {materialFunction=&MF_FX_EmissiveAdjustments}
  n16 = ComponentMask(None=TextureCoordinate()) {r=true}
  n17 = Multiply(A=Power(Base=OneMinus(None=@n16), Exp=Constant() {r=8}), B=Constant() {r=10})
  n18 = TextureSample(UVs=Panner(Coordinate=TextureCoordinate() {uTiling=2, vTiling=0.25}) {speedX=-1, speedY=-0.1}) {texture=&T_VCVFX_Boost_Anisotropic_A}
  n19 = Constant() {r=1.5}
  n20 = TextureSample(UVs=Panner(Coordinate=TextureCoordinate() {vTiling=0.5}) {speedX=-1, speedY=-0.1}) {texture=&T_VCVFX_Boost_Anisotropic_A}
  n21 = Constant() {r=1.5}
  MaterialExpressionConstantClamp_9 = Saturate(None=Add(A=Multiply(A=@n18.RGB, B=@n19), B=Multiply(A=@n20.RGB, B=@n21)))
  MaterialExpressionConstantClamp_8 = Saturate(None=Add(A=@n17, B=Multiply(A=@n17, B=@MaterialExpressionConstantClamp_9)))
  n22 = Power(Base=ComponentMask(None=TextureCoordinate()) {r=true}, Exp=Constant() {r=2})
  n23 = Constant() {r=250}
  MaterialExpressionConstantClamp_7 = Saturate(None=Multiply(A=@MaterialExpressionConstantClamp_8, B=Multiply(A=@n22, B=@n23)))
  Multiply(A=@n15, B=@MaterialExpressionConstantClamp_7)
  n24 = ComponentMask(None=TextureCoordinate()) {r=true}
  Reroute(None=Multiply(A=@n14, B=OneMinus(None=@n24)))
 ]
}
